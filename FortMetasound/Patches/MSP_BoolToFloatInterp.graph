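# MetaSound patch that turns a Bool input into a smoothly interpolated Float between configurable False and True values.
# The Bool converts to a Float, which becomes the target of an InterpTo node timed by Interp Time.
# A MapRange node maps that interpolated value onto the Value.False to Value.True range and feeds Output.
# By default, False gives one and True gives zero, so the result is inverted.
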MetaSoundPatch /FortMetasound/Patches/MSP_BoolToFloatInterp
inputs [
 Input {type=Bool}
 Interp Time {type=Time}
 Value.False {type=Float}
 Value.True {type=Float}
]
outputs [Output {type=Float}]
pages [Default {
  input_defaults {Input=null, Interp Time=0.1, Value.False=1, Value.True=0}
  nodes [MapRangeMapRange_0 = MapRange.MapRange:Float(
    Out Range A=^Value.False
    Out Range B=^Value.True
    In=UE.InterpTo:Audio(Interp Time=^Interp Time, Target=Convert.Float:Bool(Bool=^Input).Float).Value
   )]
  links [{dst="@Output", src=MapRangeMapRange_0.Out Value}]
 }]
