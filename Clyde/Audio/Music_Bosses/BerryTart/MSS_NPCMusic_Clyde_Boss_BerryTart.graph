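# MetaSound music source for the Clyde BerryTart boss, switching between idle and combat loops on phase, combat and targeting events.
# A 143 BPM metronome clocks two quantized TrackPlayback patches, mixed to stereo with an envelope on combat.
# Distance lowers idle drum gain and filter cutoff, while optional occlusion logic adjusts drum gain.
# Phases 2-3 lack playback patches, and activation and death stinger players drive ducking, stops and finish without reaching the mixer.
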MetaSoundSource /Clyde/Audio/Music_Bosses/BerryTart/MSS_NPCMusic_Clyde_Boss_BerryTart
inputs [
 ActiveSound.bIsOccluded {type=Bool}
 ActiveSound.MaxDistance {type=Float}
 Debug {type=Bool}
 Mixing.DrumMult.Phase0 {type=Float}
 Mixing.DrumMult.Phase1 {type=Float}
 Mixing.DrumMult.Phase2 {type=Float}
 Mixing.DrumMult.Phase3 {type=Float}
 Mixing.Phase0 {type=Float}
 Mixing.Phase1 {type=Float}
 Mixing.Phase2 {type=Float}
 Mixing.Phase3 {type=Float}
 Mixing.Stinger.Activate {type=Float}
 Mixing.Stinger.Death {type=Float}
 Mixing.Stinger.Phase0 {type=Float}
 Mixing.Stinger.Phase1 {type=Float}
 Mixing.Stinger.Phase2 {type=Float}
 Mixing.Stinger.Phase3 {type=Float}
 NPC.Event.OnActivate {type=Trigger}
 NPC.Event.OnMusicStop {type=Trigger}
 NPC.Event.OnPhase0 {type=Trigger}
 NPC.Event.OnPhase1 {type=Trigger}
 NPC.Event.OnPhase2 {type=Trigger}
 NPC.Event.OnPhase3 {type=Trigger}
 NPC.Event.OnPlayerLost {type=Trigger}
 NPC.Event.OnPlayerSpotted {type=Trigger}
 NPC.Music.bEnableOcclusion {type=Bool}
 NPC.Music.Metronome {type=Float}
 NPC.Music.StartingPhase {type=Float}
 NPC.Music.VolumeRadius {type=Float}
 NPC.State.bInCombat {type=Bool}
 NPC.State.bListenForTargeting {type=Bool}
 Quantize.Phase0.LoopBeatDivision {type=Int32}
 Quantize.Phase0.StingerBeatDivision {type=Int32}
 Quantize.Phase1.LoopBeatDivision {type=Int32}
 Quantize.Phase1.StingerBeatDivision {type=Int32}
 Quantize.Phase2.LoopBeatDivision {type=Int32}
 Quantize.Phase2.StingerBeatDivision {type=Int32}
 Quantize.Phase3.LoopBeatDivision {type=Int32}
 Quantize.Phase3.StingerBeatDivision {type=Int32}
 Quantize.StingerDeath.StingerBeatDivision {type=Int32}
 UE.Attenuation.Distance {type=Float}
 UE.Source.OnPlay {type=Trigger}
 UE.Spatialization.Azimuth {type=Float}
 UE.Spatialization.Elevation {type=Float}
 Wave.Phase0.DrumLoop {type=WaveAsset}
 Wave.Phase0.ElseLoop {type=WaveAsset}
 Wave.Phase0.ExtLoop {type=WaveAsset}
 Wave.Phase1.DrumLoop {type=WaveAsset}
 Wave.Phase1.ElseLoop {type=WaveAsset}
 Wave.Phase1.ExtLoop {type=WaveAsset}
 Wave.Phase2.DrumLoop {type=WaveAsset}
 Wave.Phase2.ElseLoop {type=WaveAsset}
 Wave.Phase2.ExtLoop {type=WaveAsset}
 Wave.Phase3.DrumLoop {type=WaveAsset}
 Wave.Phase3.ElseLoop {type=WaveAsset}
 Wave.Phase3.ExtLoop {type=WaveAsset}
 Wave.Stinger.Activation {type=WaveAsset}
 Wave.Stinger.Death {type=WaveAsset}
 Wave.Stinger.Transition.Phase0 {type=WaveAsset}
 Wave.Stinger.Transition.Phase1 {type=WaveAsset}
 Wave.Stinger.Transition.Phase2 {type=WaveAsset}
 Wave.Stinger.Transition.Phase3 {type=WaveAsset}
]
interfaces [Fort.ActiveSound:1.0, UE.Attenuation:1.0, UE.OutputFormat.Stereo:1.0, UE.Source.OneShot:1.0, UE.Source:1.1, UE.Spatialization:1.0]
outputs [
 "UE.OutputFormat.Stereo.Audio:0" {type=Audio}
 "UE.OutputFormat.Stereo.Audio:1" {type=Audio}
 UE.Source.OneShot.OnFinished {type=Trigger}
]
pages [Default {
  input_defaults {
   ActiveSound.MaxDistance=1
   ActiveSound.bIsOccluded=null
   Debug=null
   Mixing.DrumMult.Phase0=1
   Mixing.DrumMult.Phase1=1.4
   Mixing.DrumMult.Phase2=1
   Mixing.DrumMult.Phase3=1
   Mixing.Phase0=1
   Mixing.Phase1=1
   Mixing.Phase2=1.6
   Mixing.Phase3=1.6
   Mixing.Stinger.Activate=1
   Mixing.Stinger.Death=1.3
   Mixing.Stinger.Phase0=1
   Mixing.Stinger.Phase1=1.7
   Mixing.Stinger.Phase2=1
   Mixing.Stinger.Phase3=1
   NPC.Event.OnActivate=null
   NPC.Event.OnMusicStop=null
   NPC.Event.OnPhase0=null
   NPC.Event.OnPhase1=null
   NPC.Event.OnPhase2=null
   NPC.Event.OnPhase3=null
   NPC.Event.OnPlayerLost=null
   NPC.Event.OnPlayerSpotted=null
   NPC.Music.Metronome=143
   NPC.Music.StartingPhase=0
   NPC.Music.VolumeRadius=15000
   NPC.Music.bEnableOcclusion=null
   NPC.State.bInCombat=null
   NPC.State.bListenForTargeting="<MetaSoundBuilderResult.SUCCEEDED: 0>"
   Quantize.Phase0.LoopBeatDivision=1
   Quantize.Phase0.StingerBeatDivision=1
   Quantize.Phase1.LoopBeatDivision=1
   Quantize.Phase1.StingerBeatDivision=1
   Quantize.Phase2.LoopBeatDivision=1
   Quantize.Phase2.StingerBeatDivision=1
   Quantize.Phase3.LoopBeatDivision=1
   Quantize.Phase3.StingerBeatDivision=1
   Quantize.StingerDeath.StingerBeatDivision=1
   UE.Attenuation.Distance=4480.2202
   UE.Source.OnPlay=null
   UE.Spatialization.Azimuth=0
   UE.Spatialization.Elevation=0
   Wave.Phase0.DrumLoop=&/MX_IP_BerryTart_Boss_Idle_lp_bpm143
   Wave.Phase0.ElseLoop=null
   Wave.Phase0.ExtLoop=null
   Wave.Phase1.DrumLoop=&/MX_IP_BerryTart_Boss_Combat_lp_bpm143
   Wave.Phase1.ElseLoop=null
   Wave.Phase1.ExtLoop=null
   Wave.Phase2.DrumLoop=null
   Wave.Phase2.ElseLoop=null
   Wave.Phase2.ExtLoop=null
   Wave.Phase3.DrumLoop=null
   Wave.Phase3.ElseLoop=null
   Wave.Phase3.ExtLoop=null
   Wave.Stinger.Activation=null
   Wave.Stinger.Death=null
   Wave.Stinger.Transition.Phase0=null
   Wave.Stinger.Transition.Phase1=null
   Wave.Stinger.Transition.Phase2=null
   Wave.Stinger.Transition.Phase3=null
  }
  nodes [
   n1 = /FortMetasound/Patches/MSP_SwitchOnInt_4(Compare=^UE.Source.OnPlay, Int=Convert.Int32:Float(Float=^NPC.Music.StartingPhase).Int32)
   TriggerOnValueChange_5 = UE.Trigger On Value Change:Bool(Value=^NPC.State.bInCombat)
   TriggerAnyTriggerAny2_3 = "TriggerAny.Trigger Any (2)"(In 1=@n1.Out 3, In 0=@TriggerOnValueChange_5.Trigger)
   n2 = TriggerCompare.Clamp:Bool(A=^NPC.State.bInCombat, Compare=@TriggerAnyTriggerAny2_3.out) {also {defaults {B="<MetaSoundBuilderResult.SUCCEEDED: 0>"}}}
   TriggerAnyTriggerAny2_5 = "TriggerAny.Trigger Any (2)"(In 1=^NPC.Event.OnPlayerSpotted, In 0=@n2.TRUE)
   n3 = VariableDeferredAccessor.Int32(Variable=InitVariable.Int32().Variable)
   n4 = &MSP_SwitchOnInt_4(
    Compare=TriggerCompare.Clamp:Bool(A=^NPC.State.bListenForTargeting, Compare=@TriggerAnyTriggerAny2_5.out) {also {defaults {B="<MetaSoundBuilderResult.SUCCEEDED: 0>"}}}.TRUE
    Int=@n3.Value
   )
   TriggerAnyTriggerAny3_0 = "TriggerAny.Trigger Any (3)"(In 1=^NPC.Event.OnPhase1, In 2=@n4.Out 1, In 0=@n1.Out 1)
   n5 = InitVariable.Trigger()
   n6 = VariableDeferredAccessor.Trigger(Variable=@n5.Variable)
   n7 = VariableDeferredAccessor.Trigger(Variable=@n6.Variable)
   n8 = VariableAccessor.Trigger(Variable=VariableMutator.Trigger(Value=@TriggerAnyTriggerAny3_0.out, Variable=@n7.Variable).Variable)
   n9 = VariableAccessor.Trigger(Variable=@n8.Variable)
   n10 = VariableAccessor.Trigger(Variable=@n9.Variable)
   n11 = VariableAccessor.Trigger(Variable=@n10.Variable)
   n12 = VariableAccessor.Trigger(Variable=@n11.Variable)
   n13 = VariableAccessor.Trigger(Variable=@n12.Variable)
   ADSREnvelopeADSREnvelope_0 = ADSR Envelope.ADSR Envelope:Audio(Trigger Attack=@n13.Value) {also {defaults {Attack Time=0, Decay Curve=2, Decay Time=18, Sustain Level=0.8}}}
   n14 = VariableAccessor.Trigger(Variable=VariableMutator.Trigger(
     Variable=InitVariable.Trigger().Variable
     Value=/FortNPCMusic/Audio/Patches/MSP_NPCMusic_Metronome(BPM=^NPC.Music.Metronome, UE.Source.OnPlay=^UE.Source.OnPlay).OnBeat
    ).Variable)
   n15 = VariableMutator.Trigger(Value=^NPC.Event.OnMusicStop, Variable=InitVariable.Trigger().Variable)
   WavePlayer_0 = UE.Wave Player:Stereo(
    Wave Asset=^Wave.Stinger.Death
    Play=/FortNPCMusic/Audio/Patches/MSP_NPCMusic_Quantizer(
     Debug=^Debug
     QuantizeDivision=^Quantize.StingerDeath.StingerBeatDivision
     OnBeat=@n14.Value
     QueueTrigger=VariableAccessor.Trigger(Variable=@n15.Variable).Value
    ) {also {defaults {Identifier=End Track: Death}}}.OnTriggerQuantized
   )
   n16 = VariableMutator.Trigger(Variable=InitVariable.Trigger().Variable, Value=@WavePlayer_0.On Play)
   n17 = VariableAccessor.Trigger(Variable=VariableMutator.Trigger(Variable=InitVariable.Trigger().Variable, Value=VariableAccessor.Trigger(Variable=@n16.Variable).Value).Variable)
   n18 = VariableAccessor.Trigger(Variable=@n17.Variable)
   n19 = VariableAccessor.Trigger(Variable=@n18.Variable)
   n20 = VariableAccessor.Trigger(Variable=@n19.Variable)
   n21 = VariableAccessor.Trigger(Variable=VariableAccessor.Trigger(Variable=@n20.Variable).Variable)
   TriggerAnyTriggerAny2_1 = "TriggerAny.Trigger Any (2)"(In 1=^UE.Source.OnPlay, In 0=@n21.Value)
   TriggerControl_2 = UE.Trigger Control(Open=@TriggerAnyTriggerAny2_1.out, Trigger In=VariableAccessor.Trigger(Variable=@n13.Variable).Value, Close=@n6.Value)
   TriggerAnyTriggerAny2_2 = "TriggerAny.Trigger Any (2)"(In 1=^NPC.Event.OnPhase2, In 0=^UE.Source.OnPlay)
   TriggerRouteTriggerRouteInt324_1 = "TriggerRoute.Trigger Route (Int32, 4):Int32"(
    Set 3=^NPC.Event.OnMusicStop
    Set 1=^NPC.Event.OnPhase0
    Set 2=^NPC.Event.OnPhase3
    Value 1=^Quantize.Phase0.StingerBeatDivision
    Value 0=^Quantize.Phase2.StingerBeatDivision
    Value 2=^Quantize.Phase3.StingerBeatDivision
    Value 3=^Quantize.StingerDeath.StingerBeatDivision
    Set 0=@TriggerAnyTriggerAny2_2.out
   )
   TriggerRouteTriggerRouteFloat6_1 = "TriggerRoute.Trigger Route (Float, 6):Float"(Set 0=^UE.Source.OnPlay) {also {defaults {Value 0=1, Value 1=0, Value 2=0, Value 3=0, Value 4=0, Value 5=0}}}
   n22 = VariableAccessor.Float(Variable=VariableMutator.Float(
     Variable=InitVariable.Float().Variable
     Value=UE.InterpTo:Audio(Target=@TriggerRouteTriggerRouteFloat6_1.Value) {also {defaults {Interp Time=1}}}.Value
    ).Variable)
   n23 = VariableDeferredAccessor.Trigger(Variable=InitVariable.Trigger().Variable)
   TriggerAnyTriggerAny2_7 = "TriggerAny.Trigger Any (2)"(In 0=^UE.Source.OnPlay, In 1=@n23.Value)
   n24 = VariableDeferredAccessor.Trigger(Variable=InitVariable.Trigger().Variable)
   n25 = VariableDeferredAccessor.Trigger(Variable=InitVariable.Trigger().Variable)
   TriggerRouteTriggerRouteFloat4_0 = "TriggerRoute.Trigger Route (Float, 4):Float"(
    Value 3=^Mixing.DrumMult.Phase0
    Value 0=^Mixing.DrumMult.Phase1
    Value 1=^Mixing.DrumMult.Phase2
    Value 2=^Mixing.DrumMult.Phase3
    Set 3=@TriggerAnyTriggerAny2_7.out
    Set 0=VariableDeferredAccessor.Trigger(Variable=@n5.Variable).Value
    Set 1=@n24.Value
    Set 2=@n25.Value
   )
   n26 = UE.InterpTo:Audio(Target=@TriggerRouteTriggerRouteFloat4_0.Value) {also {defaults {Interp Time=1}}}
   n27 = VariableDeferredAccessor.Trigger(Variable=InitVariable.Trigger().Variable)
   n28 = VariableDeferredAccessor.Trigger(Variable=InitVariable.Trigger().Variable)
   TriggerRouteTriggerRouteFloat6_3 = "TriggerRoute.Trigger Route (Float, 6):Float"(Set 0=^UE.Source.OnPlay, Set 2=@n27.Value, Set 1=@n28.Value) {also {defaults {Value 0=1, Value 1=0.1, Value 2=1, Value 3=0, Value 4=0, Value 5=0}}}
   n29 = UE.InterpTo:Audio(Target=@TriggerRouteTriggerRouteFloat6_3.Value) {also {defaults {Interp Time=1}}}
   n30 = VariableAccessor.Float(Variable=VariableMutator.Float(Variable=InitVariable.Float().Variable, Value=UE.Multiply:Float(AdditionalOperands=@n26.Value, PrimaryOperand=@n29.Value).out).Variable)
   TriggerRouteTriggerRouteFloat6_0 = "TriggerRoute.Trigger Route (Float, 6):Float"(Set 0=^UE.Source.OnPlay) {also {defaults {Value 0=1, Value 1=0, Value 2=0, Value 3=0, Value 4=0, Value 5=0}}}
   n31 = VariableAccessor.Float(Variable=VariableMutator.Float(
     Variable=InitVariable.Float().Variable
     Value=UE.InterpTo:Audio(Target=@TriggerRouteTriggerRouteFloat6_0.Value) {also {defaults {Interp Time=1}}}.Value
    ).Variable)
   n32 = VariableAccessor.Trigger(Variable=@n21.Variable)
   n33 = VariableAccessor.Trigger(Variable=@n14.Variable)
   n34 = VariableAccessor.Trigger(Variable=@n33.Variable)
   n35 = VariableDeferredAccessor.Trigger(Variable=InitVariable.Trigger().Variable)
   n36 = /Clyde/Audio/Music_Bosses/AudioData/MSP_NPCMusic_ClydeBoss_TrackPlayback(
    Debug=^Debug
    Gain.Stinger=^Mixing.Stinger.Phase1
    PhaseQuantizeDivision=^Quantize.Phase1.LoopBeatDivision
    StingerQuantizeDivision=^Quantize.Phase1.StingerBeatDivision
    Wave.DrumsLoop=^Wave.Phase1.DrumLoop
    Wave.ElseLoop=^Wave.Phase1.ElseLoop
    Wave.ExtLoop=^Wave.Phase1.ExtLoop
    Wave.TransitionInStinger=^Wave.Stinger.Transition.Phase1
    OnPlay=@TriggerControl_2.Trigger Out
    OnStopQuantizationDivision=@TriggerRouteTriggerRouteInt324_1.Value
    Gain.Ext=VariableAccessor.Float(Variable=@n22.Variable).Value
    Gain.Drums=VariableAccessor.Float(Variable=@n30.Variable).Value
    Gain.Else=VariableAccessor.Float(Variable=@n31.Variable).Value
    OnStopAll=VariableAccessor.Trigger(Variable=@n32.Variable).Value
    QueueTrigger=@n11.Value
    OnBeat=@n34.Value
    ResetAndStop=@n35.Value
   ) {also {defaults {Fade In Time=1.5, Fade Out Time=1.5, bCrossfade="<MetaSoundBuilderResult.SUCCEEDED: 0>"}}}
   n37 = VariableDeferredAccessor.Float(Variable=InitVariable.Float().Variable)
   LinearToLogFrequency_0 = UE.Linear To Log Frequency(Value=@n37.Value) {also {defaults {Min Domain=0.5, Min Range=250}}}
   TriggerAnyTriggerAny2_8 = "TriggerAny.Trigger Any (2)"(In 0=^UE.Source.OnPlay, In 1=@TriggerControl_2.Trigger Out)
   TriggerRouteTriggerRouteInt325_0 = "TriggerRoute.Trigger Route (Int32, 5):Int32"(
    Set 4=^NPC.Event.OnMusicStop
    Set 1=^NPC.Event.OnPhase1
    Set 2=^NPC.Event.OnPhase2
    Set 3=^NPC.Event.OnPhase3
    Value 1=^Quantize.Phase1.StingerBeatDivision
    Value 2=^Quantize.Phase2.StingerBeatDivision
    Value 3=^Quantize.Phase3.StingerBeatDivision
    Value 4=^Quantize.StingerDeath.StingerBeatDivision
    Set 0=^UE.Source.OnPlay
   ) {also {defaults {Value 0=2}}}
   TriggerAnyTriggerAny4_0 = "TriggerAny.Trigger Any (4)"(
    In 2=^NPC.Event.OnPhase0
    In 0=@n1.Out 0
    In 1=TriggerCompare.Clamp:Bool(A=^NPC.State.bListenForTargeting, Compare=@n2.FALSE) {also {defaults {B="<MetaSoundBuilderResult.SUCCEEDED: 0>"}}}.TRUE
    In 3=TriggerCompare.Clamp:Bool(Compare=^NPC.Event.OnPlayerLost, A=^NPC.State.bListenForTargeting) {also {defaults {B="<MetaSoundBuilderResult.SUCCEEDED: 0>"}}}.TRUE
   )
   n38 = VariableAccessor.Trigger(Variable=VariableMutator.Trigger(Value=@TriggerAnyTriggerAny4_0.out, Variable=@n23.Variable).Variable)
   n39 = VariableDeferredAccessor.Trigger(Variable=InitVariable.Trigger().Variable)
   n40 = &MSP_NPCMusic_ClydeBoss_TrackPlayback(
    Debug=^Debug
    Gain.Stinger=^Mixing.Stinger.Phase0
    PhaseQuantizeDivision=^Quantize.Phase0.LoopBeatDivision
    StingerQuantizeDivision=^Quantize.Phase0.StingerBeatDivision
    Wave.DrumsLoop=^Wave.Phase0.DrumLoop
    Wave.ElseLoop=^Wave.Phase0.ElseLoop
    Wave.ExtLoop=^Wave.Phase0.ExtLoop
    Wave.TransitionInStinger=^Wave.Stinger.Transition.Phase0
    LPF.Drums=@LinearToLogFrequency_0.Frequency
    Gain.Drums=UE.Multiply:Float(PrimaryOperand=@n30.Value, AdditionalOperands=@n37.Value).out
    OnPlay=@TriggerAnyTriggerAny2_8.out
    OnStopQuantizationDivision=@TriggerRouteTriggerRouteInt325_0.Value
    Gain.Else=@n31.Value
    Gain.Ext=@n22.Value
    OnBeat=@n33.Value
    QueueTrigger=@n38.Value
    OnStopAll=@n32.Value
    ResetAndStop=@n39.Value
   ) {also {defaults {Fade In Time=4, Fade Out Time=1.5, Identifier=New Track: Idle, bCrossfade="<MetaSoundBuilderResult.SUCCEEDED: 0>"}}}
   n41 = /FortMetasound/Patches/IsWaveAssetValid(
    WaveAsset=^Wave.Stinger.Activation
    Trigger In=&MSP_NPCMusic_Quantizer(QueueTrigger=^NPC.Event.OnActivate, OnBeat=VariableAccessor.Trigger(Variable=@n34.Variable).Value) {also {defaults {Identifier=Start Track: Activate, QuantizeDivision=2}}}.OnTriggerQuantized
   )
   n42 = Convert.Int32:Float(Float=^NPC.Music.StartingPhase)
   TriggerOnValueChange_7 = UE.Trigger On Value Change:Int32(Value=@n42.Int32)
   n43 = VariableDeferredAccessor.Trigger(Variable=@n25.Variable)
   n44 = VariableDeferredAccessor.Trigger(Variable=@n24.Variable)
   TriggerRouteTriggerRouteInt324_0 = "TriggerRoute.Trigger Route (Int32, 4):Int32"(Value 0=@n42.Int32, Set 0=@TriggerOnValueChange_7.Trigger, Set 3=@n43.Value, Set 2=@n44.Value, Set 1=@n7.Value) {also {defaults {Value 1=1, Value 2=2, Value 3=3}}}
   n45 = VariableMutator.Int32(Value=@TriggerRouteTriggerRouteInt324_0.Value, Variable=@n3.Variable)
   WavePlayer_1 = UE.Wave Player:Stereo(
    Wave Asset=UE.Reroute(Value=^Wave.Stinger.Activation).Value
    Play=TriggerCompare.Clamp:Int32(Compare=@n41.Is Valid, A=VariableAccessor.Int32(Variable=@n45.Variable).Value).True
   )
   n46 = &MSP_Fade(In=@WavePlayer_1.On Play) {also {defaults {bInvertOutput="<MetaSoundBuilderResult.SUCCEEDED: 0>"}}}
   TriggerAnyTriggerAny3_1 = "TriggerAny.Trigger Any (3)"(In 1=^NPC.Event.OnPhase3, In 2=@n4.Out 3)
   n47 = VariableAccessor.Trigger(Variable=VariableMutator.Trigger(Value=@TriggerAnyTriggerAny3_1.out, Variable=@n43.Variable).Variable)
   n48 = VariableAccessor.Trigger(Variable=@n47.Variable)
   n49 = VariableAccessor.Trigger(Variable=@n48.Variable)
   TriggerAnyTriggerAny3_2 = "TriggerAny.Trigger Any (3)"(In 1=^NPC.Event.OnPhase2, In 2=@n4.Out 2, In 0=@n1.Out 2)
   n50 = VariableAccessor.Trigger(Variable=VariableMutator.Trigger(Value=@TriggerAnyTriggerAny3_2.out, Variable=@n44.Variable).Variable)
   n51 = VariableAccessor.Trigger(Variable=@n50.Variable)
   n52 = VariableAccessor.Trigger(Variable=@n51.Variable)
   TriggerAnyTriggerAny5_0 = "TriggerAny.Trigger Any (5)"(
    In 0=^UE.Source.OnPlay
    In 4=UE.Reroute(Value=@n41.Is Not Valid).Value
    In 3=VariableAccessor.Trigger(Variable=@n49.Variable).Value
    In 2=VariableAccessor.Trigger(Variable=@n52.Variable).Value
    In 1=@n12.Value
   )
   TriggerRouteTriggerRouteFloat3_0 = "TriggerRoute.Trigger Route (Float, 3):Float"(Set 2=@n46.On Done, Set 1=@n46.Out, Value 1=@n46.Value Out, Set 0=@TriggerAnyTriggerAny5_0.Out) {also {defaults {Value 0=1}}}
   n53 = VariableAccessor.Float(Variable=VariableMutator.Float(
     Variable=InitVariable.Float().Variable
     Value=UE.InterpTo:Audio(Target=@TriggerRouteTriggerRouteFloat3_0.Value) {also {defaults {Interp Time=1}}}.Value
    ).Variable)
   AudioMixerAudioMixerStereo2_0 = "AudioMixer.Audio Mixer (Stereo, 2)"(
    Gain 0=^Mixing.Phase0
    Gain 1=^Mixing.Phase1
    In 1 R=UE.Multiply:Audio(PrimaryOperand=@ADSREnvelopeADSREnvelope_0.Out Envelope, AdditionalOperands=@n36.Out R).Out
    In 1 L=UE.Multiply:Audio(PrimaryOperand=@ADSREnvelopeADSREnvelope_0.Out Envelope, AdditionalOperands=@n36.Out L).Out
    In 0 R=UE.Multiply:Audio by Float(PrimaryOperand=@n40.Out R, AdditionalOperands=@n53.Value).out
    In 0 L=UE.Multiply:Audio by Float(PrimaryOperand=@n40.Out L, AdditionalOperands=@n53.Value).out
   )
   GetWaveDuration_0 = "UE.Get Wave Duration: "(Wave=^Wave.Stinger.Activation)
   UE.ConversionTimeToFloat(In=@GetWaveDuration_0.Duration)
   InitVariable.Trigger()
   InitVariable.Trigger()
   InitVariable.Trigger()
   TriggerRouteTriggerRouteFloat6_2 = "TriggerRoute.Trigger Route (Float, 6):Float"(Set 0=^UE.Source.OnPlay) {also {defaults {Value 0=9000}}}
   TriggerOnValueChange_4 = UE.Trigger On Value Change:Bool(Value=^NPC.Music.bEnableOcclusion)
   TriggerAnyTriggerAny2_6 = "TriggerAny.Trigger Any (2)"(In 1=^UE.Source.OnPlay, In 0=@TriggerOnValueChange_4.Trigger)
   n54 = TriggerCompare.Clamp:Bool(A=UE.Reroute(Value=^NPC.Music.bEnableOcclusion).Value, Compare=@TriggerAnyTriggerAny2_6.out) {also {defaults {B="<MetaSoundBuilderResult.SUCCEEDED: 0>"}}}
   TriggerOnValueChange_1 = UE.Trigger On Value Change:Bool(Value=^ActiveSound.bIsOccluded)
   n55 = &MSP_Branch(Value=^ActiveSound.bIsOccluded, In=@TriggerOnValueChange_1.Trigger)
   TriggerOnValueChange_6 = UE.Trigger On Value Change:Float(Value=^UE.Attenuation.Distance)
   n56 = TriggerCompare.Clamp:Float(A=^UE.Attenuation.Distance, Compare=@TriggerOnValueChange_6.Trigger, B=VariableAccessor.Float(Variable=InitVariable.Float().Variable).Value) {also {defaults {Type=5}}}
   TriggerAccumulatorTriggerAccumulate2_0 = "TriggerAccumulator.Trigger Accumulate (2)"(In 0=@n55.TRUE, In 1=@n56.TRUE) {also {defaults {Auto Reset="<MetaSoundBuilderResult.SUCCEEDED: 0>"}}}
   TriggerAnyTriggerAny2_4 = "TriggerAny.Trigger Any (2)"(In 0=@n55.FALSE, In 1=@n56.FALSE)
   TriggerRouteTriggerRouteBool6_0 = "TriggerRoute.Trigger Route (Bool, 6):Bool"(
    Set 1=^NPC.Event.OnPlayerLost
    Set 0=^NPC.Event.OnPlayerSpotted
    Set 2=@TriggerAccumulatorTriggerAccumulate2_0.out
    Set 3=@TriggerAnyTriggerAny2_4.out
   ) {also {defaults {Value 1="<MetaSoundBuilderResult.SUCCEEDED: 0>", Value 2="<MetaSoundBuilderResult.SUCCEEDED: 0>"}}}
   TriggerRouteTriggerRouteBool4_0 = "TriggerRoute.Trigger Route (Bool, 4):Bool"(Set 0=@TriggerRouteTriggerRouteBool6_0.On Set, Value 0=@TriggerRouteTriggerRouteBool6_0.Value)
   TriggerOnValueChange_2 = UE.Trigger On Value Change:Bool(Value=@TriggerRouteTriggerRouteBool4_0.Value)
   TriggerControl_0 = UE.Trigger Control(Close=@n54.FALSE, Open=@n54.TRUE, Trigger In=@TriggerOnValueChange_2.Trigger)
   TriggerRouteTriggerRouteBool2_0 = "TriggerRoute.Trigger Route (Bool, 2):Bool"(
    Set 1=UE.Reroute(Value=@n54.FALSE).Value
    Value 0=UE.Reroute(Value=UE.Reroute(Value=@TriggerRouteTriggerRouteBool4_0.Value).Value).Value
    Set 0=@TriggerControl_0.Trigger Out
   )
   n57 = VariableAccessor.Bool(Variable=VariableMutator.Bool(Variable=InitVariable.Bool().Variable, Value=@TriggerRouteTriggerRouteBool2_0.Value).Variable)
   VariableMutator.Float(
    Variable=InitVariable.Float().Variable
    Value=/FortMetasound/Patches/MSP_BoolToFloatInterp(Value.True=@TriggerRouteTriggerRouteFloat6_2.Value, Input=@n57.Value) {also {defaults {Interp Time=3, Value.False=20000}}}.Output
   )
   n58 = UE.Multiply:Float(PrimaryOperand=^NPC.Music.VolumeRadius) {also {defaults {AdditionalOperands=0.75}}}
   VariableMutator.Float(
    Value=MapRange.MapRange:Float(
     In=^UE.Attenuation.Distance
     In Range A=UE.Multiply:Float(PrimaryOperand=@n58.out) {also {defaults {AdditionalOperands=0.5}}}.out
     In Range B=@n58.out
    ) {also {defaults {Out Range A=1, Out Range B=0.1}}}.Out Value
    Variable=@n37.Variable
   )
   n59 = /FortMetasound/Patches/InRangeFloat(Input=^UE.Spatialization.Azimuth) {also {defaults {Max=3, Min=-3}}}
   TriggerOnValueChange_0 = UE.Trigger On Value Change:Bool(Value=@n59.Result)
   n60 = VariableAccessor.Bool(Variable=@n57.Variable)
   TriggerOnValueChange_3 = UE.Trigger On Value Change:Bool(Value=@n60.Value)
   n61 = TriggerCompare.Clamp:Bool(Compare=@TriggerOnValueChange_3.Trigger, A=@n60.Value) {also {defaults {B="<MetaSoundBuilderResult.SUCCEEDED: 0>"}}}
   TriggerControl_1 = UE.Trigger Control(
    Trigger In=@TriggerOnValueChange_0.Trigger
    Open=VariableAccessor.Trigger(Variable=VariableMutator.Trigger(Value=@n61.FALSE, Variable=@n27.Variable).Variable).Value
    Close=VariableAccessor.Trigger(Variable=VariableMutator.Trigger(Value=@n61.TRUE, Variable=@n28.Variable).Variable).Value
   )
   n62 = TriggerCompare.Clamp:Bool(A=UE.Reroute(Value=UE.Reroute(Value=@n59.Result).Value).Value, Compare=@TriggerControl_1.Trigger Out) {also {defaults {B="<MetaSoundBuilderResult.SUCCEEDED: 0>"}}}
   TriggerAnyTriggerAny2_0 = "TriggerAny.Trigger Any (2)"(In 1=^UE.Source.OnPlay, In 0=@n62.FALSE)
   TriggerRouteTriggerRouteFloat2_0 = "TriggerRoute.Trigger Route (Float, 2):Float"(Set 1=@TriggerAnyTriggerAny2_0.out, Set 0=@n62.TRUE) {also {defaults {Value 0=1, Value 1=0.7}}}
   VariableMutator.Float(
    Variable=InitVariable.Float().Variable
    Value=UE.InterpTo:Audio(Target=@TriggerRouteTriggerRouteFloat2_0.Value) {also {defaults {Interp Time=5}}}.Value
   )
   TriggerAnyTriggerAny4_4 = "TriggerAny.Trigger Any (4)"(In 2=@n49.Value, In 1=@n52.Value, In 0=@n10.Value, In 3=@n20.Value)
   VariableMutator.Trigger(Value=@TriggerAnyTriggerAny4_4.out, Variable=@n39.Variable)
   n63 = VariableAccessor.Trigger(Variable=@n38.Variable)
   TriggerAnyTriggerAny4_1 = "TriggerAny.Trigger Any (4)"(In 2=@n47.Value, In 3=@n17.Value, In 1=@n50.Value, In 0=@n63.Value)
   VariableMutator.Trigger(Value=@TriggerAnyTriggerAny4_1.out, Variable=@n35.Variable)
   n64 = VariableAccessor.Trigger(Variable=@n63.Variable)
   TriggerAnyTriggerAny4_2 = "TriggerAny.Trigger Any (4)"(In 2=@n48.Value, In 0=@n64.Value, In 1=@n8.Value, In 3=@n18.Value)
   VariableMutator.Trigger(Variable=InitVariable.Trigger().Variable, Value=@TriggerAnyTriggerAny4_2.out)
   TriggerAnyTriggerAny4_3 = "TriggerAny.Trigger Any (4)"(In 2=@n51.Value, In 3=@n19.Value, In 1=@n9.Value, In 0=VariableAccessor.Trigger(Variable=@n64.Variable).Value)
   VariableMutator.Trigger(Variable=InitVariable.Trigger().Variable, Value=@TriggerAnyTriggerAny4_3.out)
  ]
  links [
   {dst="@UE.OutputFormat.Stereo.Audio:0", src=AudioMixerAudioMixerStereo2_0.Out L}
   {dst="@UE.OutputFormat.Stereo.Audio:1", src=AudioMixerAudioMixerStereo2_0.Out R}
   {dst="@UE.Source.OneShot.OnFinished", src=WavePlayer_0.On Finished}
  ]
 }]
properties {
 assetImportData=&self:AssetImportData
 attenuationSettings=/Clyde/Audio/Music_Bosses/AudioData/AS_NPC_Music_Clyde_Boss_Default
 maxDistance=16000
 numChannels=2
 outputFormat=Stereo
 sampleRate=48000
 soundClassObject=&Fort_Music_NPC
 virtualizationMode=PlayWhenSilent
 volume=0.7
 assetUserData [&self:AudioAssetUserData_0]
 preEffectBusSends [{audioBus=/Clyde/Audio/Music_Bosses/LemonCart/AB_Clyde_LemonCart, maxSendDistance=1000, minSendDistance=100, minSendLevel=0, sendLevel=1}]
}
